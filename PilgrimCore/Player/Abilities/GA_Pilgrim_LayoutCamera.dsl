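; Gameplay ability that switches the player to a layout camera chosen in the Pilgrim debug settings widget.
; Event activation takes the WBP_PilgrimDebugSettings widget from the event payload's OptionalObject.
; It starts that widget's camera transition using the selected camera's class defaults, then pushes that camera mode.
; When the ability ends, it runs the parent logic and pushes the base camera mode.
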(event Ability|EventActivateAbilityFromEvent
  (params (EventData :type Gameplay%20Event%20Data%20Structure%20%28by%20ref%29))
  (bind (_eventtag _instigator _target _optionalobject) (Utilities|Struct|BreakGameplayEventData EventData))
  (bind _aswbp_pilgrim_debug_settings (Utilities|Casting|CastToWBP_PilgrimDebugSettings _optionalobject))
  (Class|WBPPilgrimDebugSettings|StartCameraTransition _aswbp_pilgrim_debug_settings (Class|GetClassDefaults (Variables|/PilgrimCore/Widgets/WBP_PilgrimDebugSettings.WBP_PilgrimDebugSettings_C|GetSelectedCamera _aswbp_pilgrim_debug_settings)))
  (Camera|PushCameraMode (Variables|/PilgrimCore/Widgets/WBP_PilgrimDebugSettings.WBP_PilgrimDebugSettings_C|GetSelectedCamera _aswbp_pilgrim_debug_settings)))

(event Ability|EventOnEndAbility
  (params (bWasCancelled :type Boolean))
  (|CallParentFunction bWasCancelled)
  (Camera|PushCameraMode (Variables|Default|GetBaseCameraMode)))
